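# Masked, lit material for the Sport76 vehicle wheel, usable on skeletal meshes and built on material attributes from MF_WheelParent_Simple.
# It blends Base Color and Emissive toward a Tire Emissive Color parameter, masked by tire RGBA mask alpha times vertex color red.
# The modified attributes then pass through MF_BaseVehicle_Damage and MF_VehicleDeRez_Transition, layering shared vehicle damage and de-rez effects.
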Material /VehicleCosmetics/Wheels/Sport76/Materials/M_Wheel_Sport76
graph {
 properties {
  bAutomaticallySetUsageInEditor=false
  bUseMaterialAttributes=true
  bUsedWithMeshDeformer=true
  bUsedWithSkeletalMesh=true
  blendMode=BLEND_Masked
  pixelDepthOffsetMode=PDOM_Legacy
  previewMesh=/Engine/Transient.MaterialEditorMeshComponent_35
  usedShadingModels=Default Lit
 }
 nodes [
  n1 = MaterialFunctionCall() {materialFunction=&MF_WheelParent_Simple}
  n2 = GetMaterialAttributes(None=@n1.MaterialAttributes) {attributeGetTypes [#1, #2]}
  n3 = VectorParameter() {group=6_Emissive Properties, parameterName=Tire Emissive Color, sortPriority=0, defaultValue=<rgba 0.4784 0.1608 0.051 1>}
  n4 = Multiply(A=ComponentMask(None=@n1.TireRGBAMask) {a=true}, B=@n1.VertexColorR)
  n5 = LinearInterpolate(A=@n2.Base Color, B=@n3.RGB, Alpha=@n4)
  n6 = LinearInterpolate(A=@n2.Emissive Color, B=@n3.RGB, Alpha=@n4)
  MaterialFunctionCall(In=MaterialFunctionCall(Material=SetMaterialAttributes!(MaterialAttributes=@n2.MaterialAttributes, Base Color=@n5, Emissive Color=@n6) {attributeSetTypes [#1, #2]}) {materialFunction=&MF_BaseVehicle_Damage}.Result) {materialFunction=&MF_VehicleDeRez_Transition}
 ]
}
